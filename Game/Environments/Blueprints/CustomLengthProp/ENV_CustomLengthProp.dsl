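; Fortnite environment prop Blueprint that builds a variable-length mesh from a base, repeated mid section and end cap along local Y.
; Construction reads mesh bounds, enforces minimum length, rounds the section count and places the end cap.
; Multiple sections become instanced meshes with optional random rotation, a snapped end cap, capsule collision, and tint from building texture data or the root material.
; On destruction, it disables instance collision and destroys the added parts.
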(event EventBeginPlay)

(event Collision|EventActorBeginOverlap
  (params (OtherActor :type Actor%20Object%20Reference)))

(event EventTick
  (params (DeltaSeconds :type Float%20%28single-precision%29)))

(event EventDestroyed
  (Collision|SetCollisionEnabled (Variables|Default|GetISM_MidSection) "NoCollision")
  (for _array_element (Utilities|Array|MakeArray (Variables|Default|GetEndCap) (Variables|Default|GetMidSection) (Variables|Default|GetISM_MidSection))
    (if (Utilities|IsValid _array_element :!class /Script/Engine.KismetSystemLibrary)
      (Components|DestroyComponent _array_element))))

(fn AddCapsuleCollision
  (bind (_min _max) (Components|StaticMesh|GetLocalBounds (Variables|Default|GetMidSection)))
  (bind _returnvalue (AddComponent|NODE_AddCapsuleComponent-0 :RelativeTransform (Math|Transform|MakeTransform (Math|Vector|MakeVector 1.0 (Math|Float|Absolute(Float) (/ (Variables|Transform|GetPropLength) 2.0)) 1.0) "0.000000,0.000000,90.000000")))
  (bind _output_get (LocalVariables|SetL_ISM_Capsule _returnvalue))
  (Components|Capsule|SetCapsuleRadius _output_get (.z _max) true)
  (Components|Capsule|SetCapsuleHalfHeight _output_get (/ (Variables|Transform|GetPropLength) 2.0) true))

(fn AddInstances
  (bind _returnvalue (AddComponent|NODE_AddInstancedStaticMeshComponent-1))
  (bind _output_get (Variables|Default|SetISM_MidSection _returnvalue))
  (Components|StaticMesh|SetStaticMesh _output_get (Variables|/Script/Engine.StaticMeshComponent|GetStaticMesh (Variables|Default|GetMidSection)))
  (for _index (range (Variables|Default|GetNumberOfSections))
    (if (Variables|Transform|GetUseRandomRotation)
      (LocalVariables|SetL_RandomRotation (Utilities|Array|RandomArrayItem (LocalVariables|GetL_AR_RandomRotations)))
      (else
        (LocalVariables|SetL_RandomRotation 0.0)))
    (Components|InstancedStaticMesh|AddInstance (Variables|Default|GetISM_MidSection) (Math|Transform|MakeTransform (Math|Vector|MakeVector 0.0 (+ (* _index (Variables|Default|GetMidBounds)) (Variables|Default|GetBaseBounds))) (Math|Rotator|MakeRotator 0.0 (LocalVariables|GetL_RandomRotation))) false)
    (Rendering|SetVisibility (Variables|Default|GetMidSection) false false)))

(fn GetMaterialFromRootMesh
  (Rendering|Material|SetMaterial (Variables|Default|GetEndCap) 0 (Material|GetMID (Variables|Debug|GetStaticMeshComponent))))

(fn GetMeshBounds
  (bind (_min _max) (Components|StaticMesh|GetLocalBounds (Variables|Default|GetEndCap)))
  (bind (_min_2 _max_2) (Components|StaticMesh|GetLocalBounds (Variables|Debug|GetStaticMeshComponent)))
  (bind (_min_3 _max_3) (Components|StaticMesh|GetLocalBounds (Variables|Default|GetMidSection)))
  (Utilities|FlowControl|Sequence
    (:then_0
      (Variables|Default|SetBaseBounds (.y _max_2)))
    (:then_1
      (Variables|Default|SetMidBounds (.y _max_3)))
    (:then_2
      (Variables|Default|SetEndBounds (.y _max)))))

(fn GetTintColorFromTextureData
  (bind _outtexturedataarray (Building|GetBuildingTextureData))
  (Variables|Default|SetTintColor (Math|Conversions|ToLinearColor(Color) (Variables|/Script/FortniteGame.BuildingTextureData|GetTintColor (Utilities|Array|Get(acopy) _outtexturedataarray)))))

(fn SetEndCapPosition
  (bind _endbounds (Variables|Default|GetEndBounds))
  (bind _proplength (Variables|Transform|GetPropLength))
  (bind _returnvalue (- _proplength _endbounds))
  (bind _returnvalue_2 (Math|Vector|MakeVector 0.0 _returnvalue))
  (Transformation|SetRelativeLocation (Variables|Default|GetEndCap) _returnvalue_2 false false)
  (Variables|Default|SetStartingEndPosition _returnvalue_2))

(fn SetNumberOfSections
  (if (not (Variables|Transform|GetRemoveMidSection))
    (Variables|Default|SetNumberOfSections (Math|Float|Round (/ (Utilities|Operators|Subtract (Variables|Transform|GetPropLength) (Variables|Default|GetBaseBounds) (Variables|Default|GetEndBounds)) (Variables|Default|GetMidBounds)))))
  (return))

(fn SetTintColorOnMeshParts
  (Utilities|FlowControl|Sequence
    (:then_0
      (for _array_element (Utilities|Array|MakeArray (Variables|Default|GetMidSection) (Variables|Default|GetEndCap))
        (Default|SetColoronSecondaryMesh (Variables|Default|GetTintColor) _array_element)))
    (:then_1
      (if (Utilities|IsValid (Variables|Default|GetISM_MidSection) :!class /Script/Engine.KismetSystemLibrary)
        (Default|SetColoronSecondaryMesh (Variables|Default|GetTintColor) (Variables|Default|GetISM_MidSection))))))

(fn UpdateEndCapPosition
  (Transformation|SetRelativeLocation (Variables|Default|GetEndCap) (Math|Vector|MakeVector (.x (Variables|Default|GetStartingEndPosition)) (- (.y (Variables|Default|GetStartingEndPosition)) (Utilities|Operators|Subtract (Variables|Transform|GetPropLength) (* (Variables|Default|GetMidBounds) (Variables|Default|GetNumberOfSections)) (Variables|Default|GetEndBounds) (Variables|Default|GetBaseBounds))) (.z (Variables|Default|GetStartingEndPosition))) false false))

(fn ConstructionScript
  (if (Variables|Transform|GetRemoveMidSection)
    (Components|StaticMesh|SetStaticMesh (Variables|Default|GetMidSection) 0)
    (Variables|Transform|SetPropLength (+ (Variables|Default|GetBaseBounds) (Variables|Default|GetEndBounds)))
    (else
      _))
  (CallFunction|GetMeshBounds)
  (if (< (Variables|Transform|GetPropLength) (Utilities|Operators|Add (Variables|Default|GetBaseBounds) (Variables|Default|GetMidBounds) (Variables|Default|GetEndBounds)))
    (Variables|Transform|SetPropLength (Utilities|Operators|Add (Variables|Default|GetBaseBounds) (Variables|Default|GetMidBounds) (Variables|Default|GetEndBounds)))
    (else
      _))
  (CallFunction|SetEndCapPosition)
  (CallFunction|SetNumberOfSections)
  (if (> (Variables|Default|GetNumberOfSections) 1)
    (CallFunction|AddInstances)
    (CallFunction|UpdateEndCapPosition)
    (CallFunction|AddCapsuleCollision)
    (Default|ValidateBuildingTextureData)
    (if (Variables|Default|GetHasTextureData)
      (Default|GetTintColorfromTextureData)
      (else
        (Default|GetMaterialfromRootMesh)))
    (Default|SetTintColoronMeshParts)
    (else
      (Transformation|SetRelativeLocation (Variables|Default|GetMidSection) (Math|Vector|MakeVector 0.0 (Variables|Default|GetBaseBounds)) false false))))

(fn ValidateBuildingTextureData
  (bind _outtexturedataarray (Building|GetBuildingTextureData))
  (Variables|Default|SetHasTextureData (> (Utilities|Array|Length _outtexturedataarray) 0))
  (return))
